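# Niagara effect for the female ExoSuit backpack cosmetic, producing a blue energy glow with swirls and a flare.
# Mesh_Swirls continuously spawns short-offset meshes that grow and fade, with animated material parameters and alpha scaled by time of day.
# VerticalFlare bursts one persistent particle that drives two sprite renderers and a cyan-blue light.
# Mesh_Swirls001, a variant inheriting source movement, is disabled, as are all mesh orientation modules.
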NiagaraSystem /BRCosmetics/Accessories/FORT_Capes/Cape_F_MED_ExoSuit/FX/NS_Exosuit_F_Backpack
dependencies {
 usedDataInterfaces [&NiagaraDataInterfaceColorCurve, &NiagaraDataInterfaceCurve]
 usedDynamicInputs []
 usedModules [&SystemState]
}
emitters {
 Mesh_Swirls {
  emitterUpdateScript [
   &EmitterState()
   &SpawnBurst_Instantaneous(Spawn Count=1, Use Spawn Probability=false) {enabled=false}
   &SpawnRate(SpawnRate=3, Use Spawn Probability=false)
  ]
  particleSpawnScript [
   /Niagara/Modules/Spawn/Initialization/InitializeParticle(
    Color=<rgba 0.1 0.1 0.1 1>
    Lifetime=4
    Mesh Scale=&RandomRangeVector(Minimum=<0.23 0.2 0.2>, Maximum=<0.17 -0.2 -0.2>)
    Module.Write Color=true
    Module.Write Lifetime=true
    Module.Write Mass=true
    Module.Write Material Random=true
    Module.Write Position=true
    Module.Write Scale=true
    Module.Write SpriteRotation=true
    Module.Write SpriteSize=true
    Sprite Rotation=&RandomRangeFloat(Minimum=0, Maximum=360)
    Sprite Size=&Vector2DFromFloat(Value=&RandomRangeFloat(Minimum=2, Maximum=3))
   )
   &SphereLocation(
    MaskSpawnGroup=false
    Non Uniform Scale=<1 5 5>
    Offset=<-5 0 0>
    Override Seed=false
    Sphere Radius=1
    Surface Only=false
    UseNonUniformScale=true
    UseOffset=true
    UseSphereOrientation=false
   )
   /Niagara/Modules/Spawn/Orientation/V2/InitialMeshOrientation(
    Orientation Axis:Vector3f=<0 0 -1>
    Orientation Vector=$Engine.Owner.SystemXAxis
    Rotation=&RandomRangeVector(Minimum=<0 -0.1 0>, Maximum=<1 0.1 0>)
    Use Orientation Vector=true
    Use Rotation Vector=true
   ) {enabled=false}
   /Niagara/Modules/Spawn/Orientation/V3/InitialMeshOrientation(
    Orientation Axis=<0 0 1>
    Orientation Coordinate Space=::World
    Orientation Vector=$Engine.Owner.SystemXAxis
    Rotation=&RandomRangeVector(Minimum=<0 -0.1 0>, Maximum=<1 0.1 0>)
    Use Orientation Vector=true
    Use Rotation Vector=true
   ) {enabled=false, module=InitialMeshOrientation001}
   &DynamicMaterialParameters(
    Index 0 Param 1=&RandomRangeFloat(Minimum=0.1, Maximum=1)
    Index 0 Param 2=&RandomRangeFloat(Minimum=0, Maximum=1)
    Index 0 Param 3=1
    Module.Param0WriteEnabled=true
    Module.Param1WriteEnabled=false
    Module.Param2WriteEnabled=false
    Module.Param3WriteEnabled=false
   )
  ]
  particleUpdateScript [
   &ParticleState()
   &SolveForcesAndVelocity(Clamp Velocity=false, Limit Acceleration=false)
   &Color(
    Color=&ColorFromCurve(ColorCurve=DataInterface() {propertyValues {
       AlphaCurve {keys [
         {arriveTangent=3.0943, interpMode=RCIM_Cubic, leaveTangent=3.0943, time=0, value=0}
         <at 0.0776 0.24>
         <at 0.3663 0.225>
         {arriveTangent=-0.355, interpMode=RCIM_Cubic, leaveTangent=-0.355, time=1, value=0}
        ]}
       BlueCurve {keys [<at 0 10 RCIM_Cubic>,<at 0.25 5>,<at 0.8 1.5>,<at 1 1>,<at 1 1 RCIM_Cubic>]}
       GreenCurve {keys [<at 0 7.889 RCIM_Cubic>,<at 0.25 3.4762>,<at 0.8 0.3111>,<at 1 0.2073>,<at 1 1 RCIM_Cubic>]}
       RedCurve {keys [<at 0 1.1827 RCIM_Cubic>,<at 0.25 0.5012>,<at 0.8 0.1405>,<at 1 0.0464>,<at 1 1 RCIM_Cubic>]}
      }})
    Scale Alpha=&FN_Select_Float_from_Time_Of_Day(Morning=0.2, Day=1, Evening=1, Night=0.5, Morning Threshold=5, Evening Threshold=20, Night Threshold=3)
   )
   &DynamicMaterialParameters(
    Index 0 Param 2=&FloatFromCurve(
     FloatCurve=DataInterface() {propertyValues {Curve {keys [<at 0 0>,<at 1 0.3>]}}}
     Scale Curve=&RandomRangeFloat(Minimum=0.3, Maximum=1)
    )
    Index 0 Param 4=&FloatFromCurve(FloatCurve=DataInterface() {propertyValues {Curve {keys [<at 0 0>,<at 0.1 0.1>,<at 1 0.2>]}}}, Scale Curve=0.2)
    Module.Param0WriteEnabled=false
    Module.Param1WriteEnabled=true
    Module.Param2WriteEnabled=false
    Module.Param3WriteEnabled=true
   ) {module=DynamicMaterialParameters001}
   &ScaleMeshSize(Scale Factor=&VectorFromFloat(Value=&FloatFromCurve(
      FloatCurve=DataInterface() {propertyValues {Curve {keys [
          {interpMode=RCIM_Cubic, leaveTangent=1.5384, tangentMode=RCTM_Break, time=0, value=0}
          {arriveTangent=0.1155, interpMode=RCIM_Cubic, tangentMode=RCTM_Break, time=1, value=0.7}
         ]}}}
      Scale Curve=&RandomRangeFloat(Minimum=0.01, Maximum=0.03)
     )))
   /Niagara/Modules/Update/Orientation/V2/UpdateMeshOrientation(
    Rotation Coordinate Space:ENiagaraMeshTransforms=::NewEnumerator3
    Rotation Rate:NiagaraFloat=&Multiply_Float(
     A=&Lerp_Float(A=-1, B=1, Alpha=$Particles.Initial.MaterialRandom)
     B=&FloatFromCurve(FloatCurve=DataInterface() {propertyValues {Curve {keys [{interpMode=RCIM_Cubic, leaveTangent=-2.1521, tangentMode=RCTM_Break, time=0, value=1}, <at 1 0>]}}})
    )
    Rotation Vector:Vector3f=<1 0 0>
   ) {enabled=false}
   /Niagara/Modules/Update/Orientation/V3/UpdateMeshOrientation(
    Rotation Coordinate Space=::NewEnumerator3
    Rotation Rate=&Multiply_Float(
     A=&Lerp_Float(A=-1, B=1, Alpha=$Particles.Initial.MaterialRandom)
     B=&FloatFromCurve(FloatCurve=DataInterface() {propertyValues {Curve {keys [{interpMode=RCIM_Cubic, leaveTangent=-2.1521, tangentMode=RCTM_Break, time=0, value=1}, <at 1 0>]}}})
    )
    Rotation Vector=<1 0 0>
   ) {enabled=false, module=UpdateMeshOrientation001}
  ]
  rendererClasses [&NiagaraMeshRendererProperties]
  renderers [{rendererClass=&NiagaraMeshRendererProperties, rendererIndex=0}]
 }
 Mesh_Swirls001 {
  bEnabled=false
  emitterUpdateScript [
   &EmitterState()
   &SpawnBurst_Instantaneous(Spawn Count=1, Use Spawn Probability=false) {enabled=false}
   &SpawnRate(SpawnRate=3, Use Spawn Probability=false)
  ]
  particleSpawnScript [
   &InitializeParticle(
    Color=<rgba 0.1 0.1 0.1 1>
    Lifetime=4
    Mesh Scale=&RandomRangeVector(Minimum=<0.23 0.2 0.2>, Maximum=<0.17 -0.2 -0.2>)
    Module.Write Color=true
    Module.Write Lifetime=true
    Module.Write Mass=true
    Module.Write Material Random=true
    Module.Write Position=true
    Module.Write Scale=true
    Module.Write SpriteRotation=true
    Module.Write SpriteSize=true
    Sprite Rotation=&RandomRangeFloat(Minimum=0, Maximum=360)
    Sprite Size=&Vector2DFromFloat(Value=&RandomRangeFloat(Minimum=2, Maximum=3))
   )
   &SphereLocation(
    MaskSpawnGroup=false
    Non Uniform Scale=<1 5 5>
    Offset=<-5 0 0>
    Override Seed=false
    Sphere Radius=1
    Surface Only=false
    UseNonUniformScale=true
    UseOffset=true
    UseSphereOrientation=false
   )
   /Niagara/Modules/Spawn/Orientation/V2/InitialMeshOrientation(
    Orientation Axis:Vector3f=<0 0 -1>
    Orientation Vector=$Engine.Owner.SystemXAxis
    Rotation=&RandomRangeVector(Minimum=<0 -0.1 0>, Maximum=<1 0.1 0>)
    Use Orientation Vector=true
    Use Rotation Vector=true
   ) {enabled=false}
   /Niagara/Modules/Spawn/Orientation/V3/InitialMeshOrientation(
    Orientation Axis=<0 0 1>
    Orientation Coordinate Space=::World
    Orientation Vector=$Engine.Owner.SystemXAxis
    Rotation=&RandomRangeVector(Minimum=<0 -0.1 0>, Maximum=<1 0.1 0>)
    Use Orientation Vector=true
    Use Rotation Vector=true
   ) {enabled=false, module=InitialMeshOrientation001}
   &DynamicMaterialParameters(
    Index 0 Param 1=&RandomRangeFloat(Minimum=0.1, Maximum=1)
    Index 0 Param 2=&RandomRangeFloat(Minimum=0, Maximum=1)
    Index 0 Param 3=1
    Module.Param0WriteEnabled=true
    Module.Param1WriteEnabled=false
    Module.Param2WriteEnabled=false
    Module.Param3WriteEnabled=false
   )
  ]
  particleUpdateScript [
   &ParticleState()
   &InheritSourceMovement(Applied Position Scale=<0.02 0.02 0.02>, ApplyForce=false, ApplyPosition=true, ApplyPositionFalloff=false)
   &InheritVelocity(Inherited Velocity Amount Scale=<-0.1 -0.1 -0.1>, Speed Limit=true, UseSpeedThreshold=false) {enabled=false}
   &SolveForcesAndVelocity(Clamp Velocity=false, Limit Acceleration=false)
   &Color(
    Color=&ColorFromCurve(ColorCurve=DataInterface() {propertyValues {
       AlphaCurve {keys [
         {arriveTangent=3.0943, interpMode=RCIM_Cubic, leaveTangent=3.0943, time=0, value=0}
         <at 0.0776 0.24>
         <at 0.3663 0.225>
         {arriveTangent=-0.355, interpMode=RCIM_Cubic, leaveTangent=-0.355, time=1, value=0}
        ]}
       BlueCurve {keys [<at 0 10 RCIM_Cubic>,<at 0.25 5>,<at 0.8 1.5>,<at 1 1>,<at 1 1 RCIM_Cubic>]}
       GreenCurve {keys [<at 0 7.889 RCIM_Cubic>,<at 0.25 3.4762>,<at 0.8 0.3111>,<at 1 0.2073>,<at 1 1 RCIM_Cubic>]}
       RedCurve {keys [<at 0 1.1827 RCIM_Cubic>,<at 0.25 0.5012>,<at 0.8 0.1405>,<at 1 0.0464>,<at 1 1 RCIM_Cubic>]}
      }})
    Scale Alpha=&FN_Select_Float_from_Time_Of_Day(Morning=0.2, Day=1, Evening=1, Night=0.5, Morning Threshold=5, Evening Threshold=20, Night Threshold=3)
   )
   &DynamicMaterialParameters(
    Index 0 Param 2=&FloatFromCurve(
     FloatCurve=DataInterface() {propertyValues {Curve {keys [<at 0 0>,<at 1 0.3>]}}}
     Scale Curve=&RandomRangeFloat(Minimum=0.3, Maximum=1)
    )
    Index 0 Param 4=&FloatFromCurve(FloatCurve=DataInterface() {propertyValues {Curve {keys [<at 0 0>,<at 0.1 0.1>,<at 1 0.2>]}}}, Scale Curve=0.2)
    Module.Param0WriteEnabled=false
    Module.Param1WriteEnabled=true
    Module.Param2WriteEnabled=false
    Module.Param3WriteEnabled=true
   ) {module=DynamicMaterialParameters001}
   &ScaleMeshSize(Scale Factor=&VectorFromFloat(Value=&FloatFromCurve(
      FloatCurve=DataInterface() {propertyValues {Curve {keys [
          {interpMode=RCIM_Cubic, leaveTangent=1.5384, tangentMode=RCTM_Break, time=0, value=0}
          {arriveTangent=0.1155, interpMode=RCIM_Cubic, tangentMode=RCTM_Break, time=1, value=0.7}
         ]}}}
      Scale Curve=&RandomRangeFloat(Minimum=0.01, Maximum=0.03)
     )))
   /Niagara/Modules/Update/Orientation/V2/UpdateMeshOrientation(
    Rotation Coordinate Space:ENiagaraMeshTransforms=::NewEnumerator3
    Rotation Rate:NiagaraFloat=&Multiply_Float(
     A=&Lerp_Float(A=-1, B=1, Alpha=$Particles.Initial.MaterialRandom)
     B=&FloatFromCurve(FloatCurve=DataInterface() {propertyValues {Curve {keys [{interpMode=RCIM_Cubic, leaveTangent=-2.1521, tangentMode=RCTM_Break, time=0, value=1}, <at 1 0>]}}})
    )
    Rotation Vector:Vector3f=<1 0 0>
   ) {enabled=false}
   /Niagara/Modules/Update/Orientation/V3/UpdateMeshOrientation(
    Rotation Coordinate Space=::NewEnumerator3
    Rotation Rate=&Multiply_Float(
     A=&Lerp_Float(A=-1, B=1, Alpha=$Particles.Initial.MaterialRandom)
     B=&FloatFromCurve(FloatCurve=DataInterface() {propertyValues {Curve {keys [{interpMode=RCIM_Cubic, leaveTangent=-2.1521, tangentMode=RCTM_Break, time=0, value=1}, <at 1 0>]}}})
    )
    Rotation Vector=<1 0 0>
   ) {enabled=false, module=UpdateMeshOrientation001}
  ]
  rendererClasses [&NiagaraMeshRendererProperties]
  renderers [{rendererClass=&NiagaraMeshRendererProperties, rendererIndex=0}]
 }
 VerticalFlare {
  emitterUpdateScript [
   &EmitterState(
    Inactive Response=::NewEnumerator1
    Life Cycle Mode=::NewEnumerator1
    Loop Behavior=::NewEnumerator1
    Loop Duration Mode=::NewEnumerator1
    UseLoopDelay=false
   )
   &SpawnBurst_Instantaneous(Spawn Count=1, Use Spawn Probability=false)
  ]
  particleSpawnScript [
   &InitializeParticle(
    Color=<rgba 5 0.0385 0 0.3282>
    Module.Write Color=true
    Module.Write Lifetime=true
    Module.Write Mass=true
    Module.Write Material Random=true
    Module.Write Position=false
    Module.Write Scale=false
    Module.Write SpriteRotation=false
    Module.Write SpriteSize=true
    Position=&SimulationPosition()
    Sprite Size=&Vector2DFromFloat(Value=40)
   )
   &SphereLocation(
    MaskSpawnGroup=false
    Offset=<-15 0 0>
    Override Seed=false
    Sphere Radius=0.33
    Surface Only=false
    UseNonUniformScale=false
    UseOffset=true
    UseSphereOrientation=false
   )
   /SetVariables(Particles.SpriteAlignment=<0 0 1>) {isSetParameters=true}
   /SetVariables(Particles.VerticalSpriteSize:Vector2f=<6 90>) {isSetParameters=true}
   /SetVariables(Particles.BrighterColor:LinearColor=<rgba 1.7741 3.4234 5 0.86>) {isSetParameters=true}
   &DynamicMaterialParameters(
    Index 0 Param 1=0.3
    Index 0 Param 2=0.0273
    Index 0 Param 3=1
    Index 1 Param 1=1
    Module.Param0WriteEnabled=true
    Module.Param0WriteEnabled1=true
    Module.Param1WriteEnabled=true
    Module.Param1WriteEnabled1=true
    Module.Param2WriteEnabled=true
    Module.Param2WriteEnabled1=true
    Module.Param3WriteEnabled=true
    Module.Param3WriteEnabled1=true
    Write Parameter Index 1=true
   )
  ]
  particleUpdateScript [
   &ParticleState(Kill Particles When Lifetime Has Elapsed=false, Let Infinitely Lived Particles Die When Emitter Deactivates=true)
   &FN_Light_v3(Color=<rgba 0 0.5333 1 1>, Create Color Attribute=true, Create Exponent Attribute=true, Create Radius Attribute=true, Exponent=2, Radius=50)
  ]
  rendererClasses [&NiagaraSpriteRendererProperties, &NiagaraLightRendererProperties]
  renderers [
   {rendererClass=&NiagaraSpriteRendererProperties}
   {rendererClass=&NiagaraSpriteRendererProperties, rendererIndex=1}
   {rendererClass=&NiagaraLightRendererProperties, rendererIndex=2}
  ]
 }
}
summary {
 systemName=NS_Exosuit_F_Backpack
 emitters [
  {emitterName=VerticalFlare, rendererClasses [&NiagaraSpriteRendererProperties, &NiagaraLightRendererProperties]}
  {emitterName=Mesh_Swirls, rendererClasses [&NiagaraMeshRendererProperties]}
  {bEnabled=false, emitterName=Mesh_Swirls001, rendererClasses [&NiagaraMeshRendererProperties]}
 ]
 userVariables [User.SkeletalMesh {
   type=NiagaraDataInterfaceSkeletalMesh
   defaultValue {
    struct=NiagaraExt_VariableValue_DataInterface
    value {dataInterface=&self:NiagaraDataInterfaceSkeletalMesh_0}
   }
  }]
}
