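# Physics asset for the flower on the F_MED ExoSuit cape cosmetic, named as its flower collision.
# It holds a single body on the dyn_flower_main_MSTR bone, with no shapes listed for it.
# It has no constraints, so nothing links that body to other bodies within the asset.
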PhysicsAsset /BRCosmetics/Accessories/FORT_Capes/Cape_F_MED_ExoSuit/Meshes/F_MED_Exosuit_Cape_FlowerCollision
bodies {dyn_flower_main_MSTR []}
constraints []
